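# Editor validator that checks the sound references on Athena glider item definitions during manual and on-save validation.
# It gathers the open, front-end open, close and thrust loop sounds, reading the front-end one through GetEditorProperty.
# Each valid reference is blocking-loaded and tested against the shared allowed classes, with that slot's specific class temporarily appended by index.
# Disallowed classes fail the asset with a message listing the allowed ones.
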EditorUtilityBlueprint /Game/Sounds/Fort_Audio_Blueprints/Utilities/Validators/AudioValidator_AthenaGliderItemDefinition
graphs [
 EventGraph {links [], nodes []}
 K2_CanValidate {nodes [
   n1 = Can Validate!() -> (then, InUsecase:EDataValidationUsecase Enum=None) {is=title K2Node_FunctionEntry}
   > Return Node!(
    execute
    ReturnValue:bool=Select!(
     None:bool=false
     Manual:bool=true
     Commandlet:bool=false
     Save:bool=true
     PreSubmit:bool=false
     Script:bool=false
     Index:EDataValidationUsecase Enum=@n1
    ) -> (ReturnValue:bool)
   )
  ]}
 K2_CanValidateAsset {nodes [
   n1 = Can Validate Asset!() -> (then, InAsset:Object Reference) {is=title K2Node_FunctionEntry}
   > n2 = Parent: Can Validate Asset!(execute, InAsset:Object Reference=@n1) -> (then, ReturnValue:bool) {is=title K2Node_CallParentFunction}
   > Return Node(ReturnValue=AND Boolean(
     A=@n2
     B=Cast To AthenaGliderItemDefinition!(Object=@n1) -> (AsAthena Glider Item Definition:Athena Glider Item Definition Object Reference, bSuccess:bool=~) {is=title K2Node_DynamicCast}.bSuccess
    ))
  ]}
 K2_ValidateLoadedAsset {nodes [
   n1 = Validate Loaded Asset!() -> (then, InAsset:Object Reference) {is=title K2Node_FunctionEntry}
   > Set Asset Passes(Asset Passes:bool=true) -> (Output_Get:bool)
   > n2 = Cast To AthenaGliderItemDefinition!(execute:Exec, Object=@n1) -> (then:Exec, CastFailed:Exec, AsAthena Glider Item Definition)
   > n3 = GetEditorProperty!(execute:Exec, Object:Object Reference=@n2, PropertyName:Name=OpenFrontEndSound) -> (then:Exec, PropertyValue:Sound Base Soft Object Reference, ReturnValue:bool)
   > n4 = For Each Loop(Array:Array of Sound Base Soft Object References=Make Array!(
     "[0]":Sound Base Soft Object Reference=Get OpenSound!(self:Athena Glider Item Definition Object Reference=@n2) -> (OpenSound:Sound Base Soft Object Reference)
     "[1]":Sound Base Soft Object Reference=@n3
     "[2]":Sound Base Soft Object Reference=Get CloseSound!(self:Athena Glider Item Definition Object Reference=@n2) -> (CloseSound:Sound Base Soft Object Reference)
     "[3]":Sound Base Soft Object Reference=Get ThrustLoopSound!(self:Athena Glider Item Definition Object Reference=@n2) -> (ThrustLoopSound:Sound Base Soft Object Reference)
    ) -> (Array:Array of Sound Base Soft Object References)) -> (Array Element:Sound Base Soft Object Reference)
   n5 = Branch(execute=@n4.Completed, Condition=Get Asset Passes() -> (Asset Passes:bool))
   > AssetPasses!(execute, self:Editor Validator Base Object Reference, InAsset:Object Reference=Get InAsset() -> (InAsset:Object Reference)) -> (then)
   > Return Node(ReturnValue:EDataValidationResult Enum=Valid)
   Return Node(execute=@n5.else, ReturnValue=Invalid)
   n6 = Set Reference(execute=@n4.LoopBody, Reference:Sound Base Soft Object Reference=@n4) -> (Output_Get:Sound Base Soft Object Reference)
   > Branch(Condition=IsValidSoftObjectReference(SoftObjectReference=@n6))
   > n7 = Sequence()
   > Set Specific Allowed Class(Specific Allowed Class:Sound Base Soft Class Reference="Get (a copy)"(
     Array:Array of Sound Base Soft Class References=Get Specific Allowed Classes() -> (Specific Allowed Classes:Array of Sound Base Soft Class References)
     Dimension 1=@n4.Array Index
    ) -> (Output:Sound Base Soft Class Reference)) -> (Output_Get:Sound Base Soft Class Reference)
   > Add(
    TargetArray:Array of Sound Base Soft Class References=Get Allowed Classes() -> (Allowed Classes:Array of Sound Base Soft Class References)
    NewItem:"Sound Base Soft Class Reference (by ref)"=Get Specific Allowed Class() -> (Specific Allowed Class:Sound Base Soft Class Reference)
   ) {is=title K2Node_CallArrayFunction}
   n8 = Sequence(execute=@n7.then_1)
   > n9 = LoadAsset_Blocking(Asset=Get Reference() -> (Reference:Sound Base Soft Object Reference))
   > Set Asset(Asset:Object Reference=@n9) -> (Output_Get:Object Reference)
   n10 = Sequence(execute=@n8.then_1)
   > Set Is Allowed Class(Is Allowed Class:bool=false) -> (Output_Get:bool)
   > n11 = For Each Loop with Break!(Exec:Exec, Array:Array of Object Soft Class References=Get Allowed Classes(), Break:Exec=@n14.then) -> (LoopBody:Exec, Array Element:Object Soft Class Reference, Array Index:int, Completed:Exec)
   Branch(execute=@n11.Completed, Condition=NOT Boolean(A=Get Is Allowed Class() -> (Is Allowed Class:bool)))
   > Set Asset Passes(Asset Passes=false)
   > n12 = For Each Loop(Array:Array of Sound Base Soft Class References=Get Allowed Classes()) -> (Array Element:Sound Base Soft Class Reference)
   K2Node_Composite_0 = Format Failure Message!(Allowed Classes:String=Get Allowed Classes String() -> (Allowed Classes String:String)) -> (Result:Text) {is=title K2Node_Composite}
   AssetFails!(
    execute=@n12.Completed
    self:Editor Validator Base Object Reference
    InAsset:Object Reference=Get InAsset()
    InMessage:"Text (by ref)"=@K2Node_Composite_0
   ) -> (then)
   n13 = "To String (SoftClassReference)"!(SoftClassReference:"Object Soft Class Reference (by ref)"=@n12) -> (ReturnValue:String)
   Set Allowed Classes String(
    execute=@n12.LoopBody
    Allowed Classes String:String=Append!(
     A=Get Allowed Classes String()
     B=" "
     C:String=Split!(InPath:String=@n13) -> (PathPart:String, FilenamePart:String, ExtensionPart:String).ExtensionPart
    ) -> (ReturnValue)
   ) -> (Output_Get:String)
   "IsA ( soft )"!(execute=@n11.LoopBody, Object:Object Reference=Get Asset() -> (Asset:Object Reference), SoftClass:Object Soft Class Reference=@n11) -> (True:Exec, False:Exec)
   > n14 = Set Is Allowed Class(Is Allowed Class=true)
   Sequence(execute=@n10.then_1)
   > Remove Item!(
    execute
    TargetArray:Array of Sound Base Soft Class References=Get Allowed Classes()
    Item:"Sound Base Soft Class Reference (by ref)"=Get Specific Allowed Class()
   ) -> (then, ReturnValue:bool)
  ]}
 Format Failure Message {
  path=K2_ValidateLoadedAsset.K2Node_Composite_0.Format Failure Message
  nodes [
   n1 = Get Asset()
   n2 = Get Class(Object=@n1) -> (ReturnValue:Object Class Reference)
   Outputs!(Result:Text=Format Text!(
     Format="LOCTEXT(\"Asset references \\'{reference}\\' (Class: {class}) which is not an allowed class. Allowed classes:{allowedclasses}\")"
     reference:String=GetDisplayName(Object=@n1)
     class:String=Get Class Display Name(Class=@n2)
     allowedclasses:String=Inputs!() -> (Allowed Classes:String)
    ) -> (Result))
  ]
 }
]
variables []
