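# Editor utility that validates selected SoundWaves against a settings object's channel count, sample rate, duration and LKFS loudness limits.
# An AudioSynesthesia LKFSNRT analyzer processes assets sequentially, each OnAnalysisComplete event running validations and starting the next.
# Results accumulate in a map and text summary.
# Finally, embedded Python writes a dated, numbered CSV to a hardcoded folder, ignoring the CSV settings, and a dialog reports passes.
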EditorUtilityBlueprint /Game/Sounds/Fort_Audio_Blueprints/Utilities/Validators/SoundWaveValidatorAssetActions
graphs [
 ConstructLoudnessNRT {nodes [
   Inputs()
   > n1 = Construct Object from Class!(
    execute
    Class=/Script/AudioSynesthesia.LKFSNRT
    self
    Settings:LKFSNRTSettings Object Reference=/Script/AudioSynesthesia.Default__LKFSNRT:DefaultLKFSNRTSettings
   ) -> (then, ReturnValue:LKFSNRT Object Reference)
   > n2 = Set LoudnessNRT(LoudnessNRT:LKFSNRT Object Reference=@n1) -> (Output_Get:LKFSNRT Object Reference)
   > Is Valid(InputObject=@n2)
   > Outputs!(OnValidNRT:Exec)
  ]}
 ContainsFailingTest {nodes [
   n1 = Break Sound Wave Validation Results!(SoundWaveValidationResults:"Sound Wave Validation Results Structure (by ref)"=Inputs!() -> (ValidationResults:Sound Wave Validation Results Structure)) -> (
    NumChannelsResult_14_#1:EAudioQATestResult Enum
    SampleRateResult_15_#2:EAudioQATestResult Enum
    DurationResult_16_#3:EAudioQATestResult Enum
    MomentaryLoudnessMaxResult_17_#4:EAudioQATestResult Enum
    ShortTermLoudnessMaxResult_18_#5:EAudioQATestResult Enum
    IntegratedResult_19_#6:EAudioQATestResult Enum
   ) {is=title K2Node_BreakStruct}
   Outputs!(ReturnValue:bool=OR Boolean!(
     A="Equal (Enum)"(A:EAudioQATestResult Enum=@n1, B:EAudioQATestResult Enum=NewEnumerator1)
     B="Equal (Enum)"(A=@n1.SampleRateResult_15_#2, B=NewEnumerator1)
     C:bool="Equal (Enum)"(A=@n1.DurationResult_16_#3, B=NewEnumerator1)
     D:bool="Equal (Enum)"(A=@n1.MomentaryLoudnessMaxResult_17_#4, B=NewEnumerator1)
     E:bool="Equal (Enum)"(A=@n1.ShortTermLoudnessMaxResult_18_#5, B=NewEnumerator1)
     F:bool="Equal (Enum)"(A=@n1.IntegratedResult_19_#6, B=NewEnumerator1)
    ) -> (ReturnValue))
  ]}
 DEBUG_PrintTestResults {nodes [
   n1 = DEBUG_PrintTestResults!() -> (then, SoundWaveValidationData:Sound Wave Validation Data Structure, SoundWaveValidationResults:Sound Wave Validation Results Structure) {is=title K2Node_FunctionEntry}
   n2 = Get LoudnessNRT() -> (LoudnessNRT:LKFSNRT Object Reference)
   n3 = Break Sound Wave Validation Results(SoundWaveValidationResults=@n1.SoundWaveValidationResults)
   n4 = Break Sound Wave Validation Data!(SoundWaveValidationData:"Sound Wave Validation Data Structure (by ref)"=@n1) -> (
    SoundWaveName_21_#7:String
    NumChannels_13_#8:int
    SampleRate_14_#9:int
    Duration_15_#10:double=~
    MomentaryLoudnessMax_16_#11:double=~
    ShortTermLoudnessMax_17_#12:double=~
    IntegratedLoudness_18_#13:double=~
   ) {is=title K2Node_BreakStruct}
   n5 = "Equal (Enum)"(A=@n3.IntegratedResult_19_#6, B=NewEnumerator1)
   n6 = Break Soundwave Validation Settings!(SoundwaveValidationSettings:"Soundwave Validation Settings Structure (by ref)"=Get Validation Settings() -> (Validation Settings:Soundwave Validation Settings Structure)) -> (
    RequiredNumChannels_53_#14:int
    RequiredSampleRate_57_#15:int
    MinDuration_56_#16:double=~
    MaxDuration_2_#17:double=~
    MinMomentaryLoudness_29_#18:double=~
    MaxMomentaryLoudness_31_#19:double=~
    MinShortTermLoudness_13_#20:double=~
    MaxShortTermLoudness_14_#21:double=~
    MinIntegratedLoudness_37_#22:double=~
    MaxIntegratedLoudness_39_#23:double=~
    MakeCsvFile_60_#24:bool=~
    CsvFolder_69_#25:String
    CsvFilenameBase_66_#26:String
   ) {is=title K2Node_BreakStruct}
   n7 = Format Text!(
    Format="LOCTEXT(\"[Target: {Target}]\")"
    Target:Text=Format Text!(
     Format="LOCTEXT(\"(Min: {min}, Max: {max})\")"
     min:double=@n6.MinIntegratedLoudness_37_#22
     max:double=@n6.MaxIntegratedLoudness_39_#23
    ) -> (Result)
   ) -> (Result)
   n8 = "Equal (Enum)"(A=@n3, B=NewEnumerator1)
   n9 = Format Text!(Format="LOCTEXT(\"[Target: {Target}]\")", Target:int=@n6) -> (Result)
   n10 = "Equal (Enum)"(A=@n3.SampleRateResult_15_#2, B=NewEnumerator1)
   n11 = Format Text(Format="LOCTEXT(\"[Target: {Target}]\")", Target=@n6.RequiredSampleRate_57_#15)
   n12 = "Equal (Enum)"(A=@n3.DurationResult_16_#3, B=NewEnumerator1)
   n13 = Format Text(
    Format="LOCTEXT(\"[Target: {Target}]\")"
    Target:Text=Format Text!(
     Format="LOCTEXT(\"(Min: {min}, Max: {max})\")"
     min:double=@n6.MinDuration_56_#16
     max:double=@n6.MaxDuration_2_#17
    ) -> (Result)
   )
   n14 = "Equal (Enum)"(A=@n3.MomentaryLoudnessMaxResult_17_#4, B=NewEnumerator1)
   n15 = Format Text!(
    Format="LOCTEXT(\"[Target: {Target}]\")"
    Target:Text=Format Text!(
     Format="LOCTEXT(\"(Min: {min}, Max: {max})\")"
     min:double=@n6.MinMomentaryLoudness_29_#18
     max:double=@n6.MaxMomentaryLoudness_31_#19
    ) -> (Result)
   ) -> (Result)
   n16 = "Equal (Enum)"(A=@n3.ShortTermLoudnessMaxResult_18_#5, B=NewEnumerator1)
   n17 = Format Text!(
    Format="LOCTEXT(\"[Target: {Target}]\")"
    Target:Text=Format Text!(
     Format="LOCTEXT(\"(Min: {min}, Max: {max})\")"
     min:double=@n6.MinShortTermLoudness_13_#20
     max:double=@n6.MaxShortTermLoudness_14_#21
    ) -> (Result)
   ) -> (Result)
   n18 = Format Text!(
    Format="LOCTEXT(\"\\r\\n-------------------------------------------------------------------------\\r\\n+ SoundWave Name: \\\"{SoundWaveName}\\\" {Result}\\r\\n+\\r\\n+ [SoundWave Info]\\r\\n{SoundWaveInfo}\\r\\n+\\r\\n+ [Test Results]\\r\\n{TestResults}\\r\\n-------------------------------------------------------------------------\\r\\n\")"
    SoundWaveName:String=GetDisplayName(Object=Get Sound!(self:Audio Analyzer NRT Object Reference=@n2) -> (Sound:Sound Wave Object Reference))
    Result:String=SelectString(
     A=(Failed)
     B=(Passed)
     bPickA=Contains Failing Test!(ValidationResults:Sound Wave Validation Results Structure=@n1.SoundWaveValidationResults) -> (ReturnValue:bool=~) {is=title K2Node_MacroInstance}
    )
    SoundWaveInfo:Text=Format Text!(
     Format="LOCTEXT(\"+ NumChannels: {NumChannels} {TargetNumChannels}\\r\\n+ SampleRate: {SampleRate} {TargetSampleRate}\\r\\n+ Duration: {Duration} {TargetDuration}\\r\\n+ LKFS-M (Max): {MomentaryMax} {TargetMomentaryMax}\\r\\n+ LKFS-S (Max): {ShortTermMax} {TargetShortTermMax}\\r\\n+ LKFS-I: {Integrated} {TargetIntegrated}\")"
     NumChannels:int=@n4.NumChannels_13_#8
     TargetNumChannels:Text=SelectText(A=@n9, bPickA=@n8)
     SampleRate:int=@n4.SampleRate_14_#9
     TargetSampleRate:Text=SelectText(A=@n11, bPickA=@n10)
     Duration:double=@n4.Duration_15_#10
     TargetDuration:Text=SelectText(A=@n13, bPickA=@n12)
     MomentaryMax:double=@n4.MomentaryLoudnessMax_16_#11
     TargetMomentaryMax:Text=SelectText(A=@n15, bPickA=@n14)
     ShortTermMax:double=@n4.ShortTermLoudnessMax_17_#12
     TargetShortTermMax:Text=SelectText(A=@n17, bPickA=@n16)
     Integrated:double=@n4.IntegratedLoudness_18_#13
     TargetIntegrated:Text=SelectText(A=@n7, bPickA=@n5)
    ) -> (Result)
    TestResults:Text=Format Text!(
     Format="LOCTEXT(\"+ NumChannels: {NumChannelsResult}\\r\\n+ SampleRate: {SampleRateResult}\\r\\n+ Duration: {DurationResult}\\r\\n+ LKFS-M (Max): {MomentaryMaxResult}\\r\\n+ LKFS-S (Max): {ShortTermMaxResult}\\r\\n+ LKFS-I: {IntegratedResult}\")"
     NumChannelsResult:String=Enum to String!(Enumerator:EAudioQATestResult Enum=@n3) -> (ReturnValue:String)
     SampleRateResult:String=Enum to String(Enumerator=@n3.SampleRateResult_15_#2)
     DurationResult:String=Enum to String(Enumerator=@n3.DurationResult_16_#3)
     MomentaryMaxResult:String=Enum to String(Enumerator=@n3.MomentaryLoudnessMaxResult_17_#4)
     ShortTermMaxResult:String=Enum to String(Enumerator=@n3.ShortTermLoudnessMaxResult_18_#5)
     IntegratedResult:String=Enum to String(Enumerator=@n3.IntegratedResult_19_#6)
    ) -> (Result)
   ) -> (Result)
   > PrintText(InText=@n18)
   > Set SoundWaveSummary(SoundWaveSummary:String=Append(A=Get SoundWaveSummary() -> (SoundWaveSummary:String), B="To String (Text)"(InText=@n18))) -> (Output_Get:String)
  ]}
 DEBUG_ShortTermMapToString {nodes [
   n1 = DEBUG_ShortTermMapToString!() -> (then, FailingIndexToShortTermLoudness:"Map of Integers to Float (double-precision)s") {is=title K2Node_FunctionEntry}
   > n2 = "For Each Loop (Map)"!(execute:Exec, MapPin:"Map of Integers to Float (double-precision)s"=@n1, BreakPin:Exec) -> (then:Exec, KeyPin:int, ValuePin:double, CompletedPin:Exec)
   n3 = "To String (Integer)"(InInt=@n2)
   n4 = "To String (Float)"(InDouble=@n2.ValuePin)
   > Set String(String:String=Append!(A=Get String() -> (String:String), B=Append!(A="(", B=@n3, C=", ", D:String=@n4, E:String=")") -> (ReturnValue), C:String="\r\n") -> (ReturnValue)) -> (Output_Get:String)
   Return Node!(execute=@n2.CompletedPin, Result:String=Get String())
  ]}
 DisplayMessage {nodes [
   DisplayMessage() {is=title K2Node_FunctionEntry}
   > n1 = GetNumPassingFiles!(
    execute
    self
    Sound Wave Validation Results:Map of Sound Wave Validation Data Structures to Sound Wave Validation Results Structures=Get Sound Wave Validation Results() -> (Sound Wave Validation Results:Map of Sound Wave Validation Data Structures to Sound Wave Validation Results Structures)
   ) -> (then, NumFailingSoundWaves:int=0)
   n2 = Get SelectedAssets() -> (SelectedAssets:Array of Object References)
   > Show Message Dialog(
    Title=Format Text!(Format="LOCTEXT(\"Sound Wave Validation\")") -> (Result)
    Message=Format Text!(
     Format="LOCTEXT(\"{NumPassingFiles}/{NumFiles} files passed validations.{FileSummary}\")"
     NumPassingFiles:int=@n1
     NumFiles:int=Length(TargetArray:Array of Object References=@n2)
     FileSummary:String=Get SoundWaveSummary()
    ) -> (Result)
    MessageType=Ok
    DefaultValue=No
    MessageCategory=Info
   )
  ]}
 EventGraph {nodes [
   n1 = ValidateSelectedSoundWaves_Event!() -> (OutputDelegate, then, ValidationSettings:BP Soundwave Validation Settings Object Reference) {is=title K2Node_CustomEvent}
   > Set Validation Settings(Validation Settings:Soundwave Validation Settings Structure=Get Settings!(self:BP Soundwave Validation Settings Object Reference=@n1.ValidationSettings) -> (Settings:Soundwave Validation Settings Structure)) -> (Output_Get:Soundwave Validation Settings Structure)
   > n2 = GetSelectedAssets()
   > Set SelectedAssets(SelectedAssets:Array of Object References=@n2) -> (Output_Get:Array of Object References)
   > Construct Loudness NRT!(execute:Exec) -> (OnValidNRT:Exec) {is=title K2Node_MacroInstance}
   > n3 = Sequence()
   > Bind Event to On Analysis Complete(self:Audio Analyzer NRT Object Reference=Get LoudnessNRT(), Delegate=Create Event()) {is=title K2Node_AddDelegate}
   Sequence(execute=@n3.then_1)
   > Branch(Condition=Is Not Empty!(TargetArray:Array of Object References=Get SelectedAssets()) -> (ReturnValue:bool))
   > SetLoudnessAnalyzerSound!(
    execute
    self
    Object:Object Reference="Get (a copy)"(Array:Array of Object References=Get SelectedAssets(), Dimension 1=Get CurrentSoundwaveIndex() -> (CurrentSoundwaveIndex:int=0)) -> (Output:Object Reference)
   ) -> (then)
   OnAnalysisComplete() {is=title K2Node_CustomEvent}
   > n4 = Sequence()
   > n5 = RunSoundwaveValidations!(execute, self) -> (then, SoundWaveValidationData:Sound Wave Validation Data Structure, SoundWaveValidationResults:Sound Wave Validation Results Structure)
   > Add!(
    execute
    TargetMap:Map of Sound Wave Validation Data Structures to Sound Wave Validation Results Structures=Get Sound Wave Validation Results()
    Key:"Sound Wave Validation Data Structure (by ref)"=@n5
    Value:"Sound Wave Validation Results Structure (by ref)"=@n5.SoundWaveValidationResults
   ) -> (then) {is=title K2Node_CallFunction}
   Sequence(execute=@n4.then_1)
   > n6 = Increment Int(Value=Get CurrentSoundwaveIndex() -> (CurrentSoundwaveIndex=0))
   n7 = Get SelectedAssets()
   > n8 = Branch(Condition="integer <= integer"(A=@n6, B=Last Index!(TargetArray:Array of Object References=@n7) -> (ReturnValue=0)))
   > SetLoudnessAnalyzerSound(Object="Get (a copy)"(Array=Get SelectedAssets(), Dimension 1=Get CurrentSoundwaveIndex() -> (CurrentSoundwaveIndex=0)))
   OnAllSelectedFilesAnalyzed(execute=@n8.else)
   OnAllSelectedFilesAnalyzed() {is=title K2Node_CustomEvent}
   > WriteTestResultsToCSV!(
    execute
    self
    SoundWaveDataToResults:Map of Sound Wave Validation Data Structures to Sound Wave Validation Results Structures=Get Sound Wave Validation Results()
   ) -> (then)
   > DisplayMessage() {is=title K2Node_CallFunction}
  ]}
 GetLoudnessDataFromNRT {nodes [
   GetIntegratedLoudness!(self:LKFSNRT Object Reference=Get LoudnessNRT()) -> (ReturnValue:float)
   GetGatedLoudness!(self:LKFSNRT Object Reference=Get LoudnessNRT()) -> (ReturnValue:float)
   GetLoudnessDataFromNRT() {is=title K2Node_FunctionEntry}
   Get LoudnessNRT()
  ]}
 GetMomentaryLoudnessData {nodes [
   n1 = GetMomentaryLoudnessData!() -> (then, LoudnessNRT:LKFSNRT Object Reference) {is=title K2Node_FunctionEntry}
   > n2 = For Each Loop(Array:Array of LKFSNRTResults Structures=GetLoudnessData!(self:LKFSNRT Object Reference=@n1) -> (ReturnValue:Array of LKFSNRTResults Structures)) -> (Array Element:LKFSNRTResults Structure)
   n3 = Get LoudnessValues() -> (LoudnessValues:"Array of Float (double-precision)s")
   Return Node!(
    execute=@n2.Completed
    MinMomentaryLoudness:float=MinOfFloatArray!(FloatArray:"Array of Float (single-precision)s"=@n3) -> (IndexOfMinValue:int=0, MinValue:float).MinValue
    MaxMomentaryLoudness:float=MaxOfFloatArray!(FloatArray:"Array of Float (single-precision)s"=@n3) -> (IndexOfMaxValue:int=0, MaxValue:float).MaxValue
   )
   > Add(
    TargetArray:"Array of Float (double-precision)s"=Get LoudnessValues()
    NewItem:"Float (double-precision) (by ref)"=Break LKFSNRTResults!(LKFSNRTResults:"LKFSNRTResults Structure (by ref)"=@n2) -> (Loudness:float=~) {is=title K2Node_BreakStruct}
   ) {is=title K2Node_CallArrayFunction}
  ]}
 GetNumPassingFiles {nodes [
   n1 = GetNumPassingFiles!() -> (then, Sound Wave Validation Results:Map of Sound Wave Validation Data Structures to Sound Wave Validation Results Structures) {is=title K2Node_FunctionEntry}
   > n2 = "For Each Loop (Map)"(MapPin:Map of Sound Wave Validation Data Structures to Sound Wave Validation Results Structures=@n1) -> (KeyPin:Sound Wave Validation Data Structure, ValuePin:Sound Wave Validation Results Structure)
   > n3 = Branch(Condition=Contains Failing Test(ValidationResults=@n2.ValuePin) -> (ReturnValue=~))
   Increment Int(" "=@n3.else, Value=Get PassingSoundWaveCount() -> (PassingSoundWaveCount:int=0))
   Return Node!(execute=@n2.CompletedPin, NumFailingSoundWaves:int=Get PassingSoundWaveCount() -> (PassingSoundWaveCount=0))
  ]}
 GetShortTermLoudnessData {nodes [
   n1 = GetShortTermLoudnessData!() -> (then, LoudnessNRT:LKFSNRT Object Reference) {is=title K2Node_FunctionEntry}
   > n2 = For Each Loop(Array=GetLoudnessData(self=@n1))
   n3 = Get LoudnessValues()
   Return Node!(
    execute=@n2.Completed
    MinShortTermLoudness:double=MinOfFloatArray(FloatArray=@n3) -> (IndexOfMinValue=0).MinValue
    MaxShortTermLoudness:double=MaxOfFloatArray(FloatArray=@n3) -> (IndexOfMaxValue=0).MaxValue
   )
   > Add(TargetArray=Get LoudnessValues(), NewItem=Break LKFSNRTResults!(LKFSNRTResults=@n2) -> (ShortTermLoudness:float=~))
  ]}
 GetSoundWaveValidationDataFromNRT {nodes [
   n1 = GetSoundWaveValidationDataFromNRT!() -> (then, LoudnessNRT:LKFSNRT Object Reference) {is=title K2Node_FunctionEntry}
   > n2 = Sequence()
   n3 = Get Sound(self=@n1)
   > Set members in Sound Wave Validation Data!(
    execute:Exec
    StructRef:"Sound Wave Validation Data Structure (by ref)"=Get ValidationInfo() -> (ValidationInfo:Sound Wave Validation Data Structure)
    SoundWaveName_21_#7:String=GetDisplayName(Object=@n3)
   ) -> (then:Exec, StructOut:"Sound Wave Validation Data Structure (by ref)") {is=title K2Node_SetFieldsInStruct}
   n4 = Sequence(execute=@n2.then_1)
   > n5 = GetEditorProperty!(execute:Exec, Object:Object Reference=@n3, PropertyName:Name=NumChannels) -> (then:Exec, PropertyValue:int, ReturnValue:bool)
   > Set members in Sound Wave Validation Data!(execute, StructRef=Get ValidationInfo(), NumChannels_13_#8:int=@n5) -> (then, StructOut)
   n6 = Sequence(execute=@n4.then_1)
   > n7 = GetEditorProperty(Object=@n3, PropertyName=SampleRate)
   > Set members in Sound Wave Validation Data!(execute, StructRef=Get ValidationInfo(), SampleRate_14_#9:int=@n7) -> (then, StructOut)
   n8 = Sequence(execute=@n6.then_1)
   > Set members in Sound Wave Validation Data!(
    execute
    StructRef=Get ValidationInfo()
    Duration_15_#10:double=Get Duration!(self:Sound Base Object Reference=@n3) -> (Duration:float)
   ) -> (then, StructOut)
   n9 = Sequence(execute=@n8.then_1)
   n10 = Get LoudnessNRT()
   > n11 = GetMomentaryLoudnessData!(execute, self, LoudnessNRT:LKFSNRT Object Reference=@n10) -> (then, MinMomentaryLoudness:float, MaxMomentaryLoudness:float) {is=title K2Node_CallFunction}
   > Set members in Sound Wave Validation Data!(
    execute
    StructRef=Get ValidationInfo()
    MomentaryLoudnessMax_16_#11:double=SelectFloat(A=-96, B=@n11.MaxMomentaryLoudness, bPickA="float < float"(A=@n11.MaxMomentaryLoudness, B=-96))
   ) -> (then, StructOut)
   n12 = Sequence(execute=@n9.then_1)
   > n13 = GetShortTermLoudnessData!(execute, self, LoudnessNRT:LKFSNRT Object Reference=@n10) -> (then, MinShortTermLoudness:double, MaxShortTermLoudness:double) {is=title K2Node_CallFunction}
   > Set members in Sound Wave Validation Data!(
    execute
    StructRef=Get ValidationInfo()
    ShortTermLoudnessMax_17_#12:double=SelectFloat(A=-96, B=@n13.MaxShortTermLoudness, bPickA="float < float"(A=@n13.MaxShortTermLoudness, B=-96))
   ) -> (then, StructOut)
   n14 = Sequence(execute=@n12.then_1)
   n15 = GetIntegratedLoudness(self=Get LoudnessNRT())
   > Set members in Sound Wave Validation Data!(
    execute
    StructRef=Get ValidationInfo()
    IntegratedLoudness_18_#13:double=SelectFloat(A=-96, B=@n15, bPickA="float < float"(A=@n15, B=-96))
   ) -> (then, StructOut)
   Return Node!(execute=@n14.then_1, SoundWaveValidationInfo:Sound Wave Validation Data Structure=Get ValidationInfo())
   Set members in Sound Wave Validation Data!(execute, StructRef=Get ValidationInfo(), Duration_15_#10:double=0) -> (then, StructOut)
  ]}
 MakeCsvFilePath {nodes [
   MakeCsvFilePath() {is=title K2Node_FunctionEntry}
   n1 = BreakDateTime!(InDateTime:Date Time Structure=Today!() -> (ReturnValue:Date Time Structure)) -> (Year:int=0, Month:int=0, Day:int=0, Hour:int=0, Minute:int=0, Second:int=0, Millisecond:int=0)
   n2 = "To Text (Integer)"(Value=@n1.Day, bUseGrouping=false, MinimumIntegralDigits=2)
   n3 = "To Text (Integer)"(Value=@n1.Month, bUseGrouping=false, MinimumIntegralDigits=2)
   n4 = "To Text (Integer)"(Value=@n1, bUseGrouping=false)
   n5 = Append!(A="D:\\Dev\\Python\\", B=Get CSVFile Name() -> (CSVFile Name:String)) -> (ReturnValue)
   > n6 = Set CSVFile Name(CSVFile Name:String="To String (Text)"(InText=Format Text!(Format="LOCTEXT(\"{Filename}_{Y}-{M}{D}\")", Filename:String=@n5, Y:Text=@n4, M:Text=@n3, D:Text=@n2) -> (Result))) -> (Output_Get:String)
   n7 = "To Text (Integer)"(Value=Get FileNumber() -> (FileNumber:int=0), bUseGrouping=false, MinimumIntegralDigits=2)
   > n8 = Set NewFileName(NewFileName:String=Append!(A=@n6, B=_, C:String="To String (Text)"(InText=@n7), D:String=.csv) -> (ReturnValue)) -> (Output_Get:String)
   > n9 = Set FileExists(FileExists:bool=FileExists!(InPath:String=@n8) -> (ReturnValue:bool)) -> (Output_Get:bool)
   > n10 = While Loop!(execute:Exec, Condition:bool=@n9) -> (LoopBody:Exec, Completed:Exec) {is=title K2Node_MacroInstance}
   Return Node!(execute=@n10.Completed, Path:String=Get NewFileName() -> (NewFileName:String))
   > n11 = Increment Int(Value=Get FileNumber() -> (FileNumber=0))
   n12 = "To Text (Integer)"(Value=@n11, MinimumIntegralDigits=2)
   > n13 = Set NewFileName(NewFileName=Append(A=Get CSVFile Name(), B=_, C="To String (Text)"(InText=@n12), D=.csv))
   > Set FileExists(FileExists=FileExists(InPath=@n13))
  ]}
 RunSoundwaveValidations {nodes [
   Sequence()
   n1 = Break Soundwave Validation Settings!(SoundwaveValidationSettings=Get Validation Settings()) -> (
    RequiredSampleRate_57_#15
    ValidateShortTermLoudness_20_#27:bool=~
    MinShortTermLoudness_13_#20=~
    MaxShortTermLoudness_14_#21=~
    MinAnalysisTime_26_#28:double=~
    MaxAnalysisTime_27_#29:double=~
    NumAllowedConsecutiveShortTermLoudnessFailures_16_#30:int
    MakeCsvFile_60_#24=~
    CsvFolder_69_#25
    CsvFilenameBase_66_#26
   )
   > Branch(Condition=@n1.ValidateShortTermLoudness_20_#27)
   > n2 = ValidateNRTLoudnessRange!(
    execute
    self
    LoudnessNRT:LKFSNRT Object Reference=Get LoudnessNRT()
    LoudnessMin:double=@n1.MinShortTermLoudness_13_#20
    LoudnessMax:double=@n1.MaxShortTermLoudness_14_#21
    MinTimeToValidate:double=@n1.MinAnalysisTime_26_#28
    MaxTimeToValidate:double=@n1.MaxAnalysisTime_27_#29
    NumAllowedConsecutiveFailures:int=@n1.NumAllowedConsecutiveShortTermLoudnessFailures_16_#30
   ) -> (
    then
    Result:EAudioQATestResult Enum=NewEnumerator3
    FramesWithFailingLoudnessValues:"Map of Float (double-precision)s to Float (double-precision)s"
   )
   RunSoundwaveValidations() {is=title K2Node_FunctionEntry}
   > n3 = GetSoundWaveValidationDataFromNRT!(execute, self, LoudnessNRT:LKFSNRT Object Reference=Get LoudnessNRT()) -> (then, SoundWaveValidationInfo:Sound Wave Validation Data Structure) {is=title K2Node_CallFunction}
   > Set ValidationData(ValidationData:Sound Wave Validation Data Structure=@n3) -> (Output_Get:Sound Wave Validation Data Structure)
   > n4 = Sequence()
   n5 = Break Soundwave Validation Settings!(SoundwaveValidationSettings=Get Validation Settings()) -> (
    ValidateNumChannels_51_#31:bool=~
    RequiredNumChannels_53_#14:int
    MakeCsvFile_60_#24=~
    CsvFolder_69_#25
    CsvFilenameBase_66_#26
   )
   > Branch(Condition=@n5)
   n6 = Break Sound Wave Validation Data!(SoundWaveValidationData=Get ValidationData() -> (ValidationData:Sound Wave Validation Data Structure)) -> (NumChannels_13_#8)
   > Set members in Sound Wave Validation Results!(
    execute:Exec
    StructRef:"Sound Wave Validation Results Structure (by ref)"=Get ValidationResults() -> (ValidationResults:Sound Wave Validation Results Structure)
    NumChannelsResult_14_#1:EAudioQATestResult Enum=Select(
     Option 0:EAudioQATestResult Enum=NewEnumerator1
     Option 1:EAudioQATestResult Enum=NewEnumerator0
     Index="Equal (Integer)"(A=@n6, B=@n5.RequiredNumChannels_53_#14)
    ) -> (ReturnValue:EAudioQATestResult Enum)
   ) -> (then:Exec, StructOut:"Sound Wave Validation Results Structure (by ref)") {is=title K2Node_SetFieldsInStruct}
   n7 = Sequence(execute=@n4.then_1)
   n8 = Break Soundwave Validation Settings!(SoundwaveValidationSettings=Get Validation Settings()) -> (
    ValidateSampleRate_42_#32:bool=~
    RequiredSampleRate_57_#15:int
    MakeCsvFile_60_#24=~
    CsvFolder_69_#25
    CsvFilenameBase_66_#26
   )
   > Branch(Condition=@n8)
   n9 = Break Sound Wave Validation Data!(SoundWaveValidationData=Get ValidationData()) -> (SampleRate_14_#9:int)
   > Set members in Sound Wave Validation Results!(
    execute
    StructRef=Get ValidationResults()
    SampleRateResult_15_#2:EAudioQATestResult Enum=Select(
     Option 0=NewEnumerator1
     Option 1=NewEnumerator0
     Index="Equal (Integer)"(A=@n9, B=@n8.RequiredSampleRate_57_#15)
    )
   ) -> (then, StructOut)
   n10 = Sequence(execute=@n7.then_1)
   n11 = Break Soundwave Validation Settings!(SoundwaveValidationSettings=Get Validation Settings()) -> (
    ValidateDuration_18_#33:bool=~
    MinDuration_56_#16:double=~
    MaxDuration_2_#17:double=~
    MakeCsvFile_60_#24=~
    CsvFolder_69_#25
    CsvFilenameBase_66_#26
   )
   > Branch(Condition=@n11)
   n12 = Break Sound Wave Validation Data!(SoundWaveValidationData=Get ValidationData()) -> (Duration_15_#10:double=~)
   > Set members in Sound Wave Validation Results!(
    execute
    StructRef=Get ValidationResults()
    DurationResult_16_#3:EAudioQATestResult Enum=Select(
     Option 0=NewEnumerator1
     Option 1=NewEnumerator0
     Index="In Range (Float)"!(
      Value:double=@n12
      Min:double=@n11.MinDuration_56_#16
      Max:double=@n11.MaxDuration_2_#17
      InclusiveMin:bool=true
      InclusiveMax:bool=true
     ) -> (ReturnValue:bool)
    )
   ) -> (then, StructOut)
   n13 = Sequence(execute=@n10.then_1)
   n14 = Break Soundwave Validation Settings!(SoundwaveValidationSettings=Get Validation Settings()) -> (
    ValidateMomentaryLoudness_22_#34:bool=~
    MinMomentaryLoudness_29_#18:double=~
    MaxMomentaryLoudness_31_#19:double=~
   )
   > Branch(Condition=@n14)
   n15 = Break Sound Wave Validation Data!(SoundWaveValidationData=Get ValidationData()) -> (MomentaryLoudnessMax_16_#11:double=~)
   > Set members in Sound Wave Validation Results!(
    execute
    StructRef=Get ValidationResults()
    MomentaryLoudnessMaxResult_17_#4:EAudioQATestResult Enum=Select(
     Option 0=NewEnumerator1
     Option 1=NewEnumerator0
     Index="In Range (Float)"(
      Value=@n15
      Min=@n14.MinMomentaryLoudness_29_#18
      Max=@n14.MaxMomentaryLoudness_31_#19
      InclusiveMin=true
      InclusiveMax=true
     )
    )
   ) -> (then, StructOut)
   n16 = Sequence(execute=@n13.then_1)
   n17 = Break Soundwave Validation Settings!(SoundwaveValidationSettings=Get Validation Settings()) -> (
    ValidateShortTermLoudness_20_#27:bool=~
    MinShortTermLoudness_13_#20:double=~
    MaxShortTermLoudness_14_#21:double=~
   )
   > Branch(Condition=@n17)
   n18 = Break Sound Wave Validation Data!(SoundWaveValidationData=Get ValidationData()) -> (ShortTermLoudnessMax_17_#12:double=~)
   > Set members in Sound Wave Validation Results!(
    execute
    StructRef=Get ValidationResults()
    ShortTermLoudnessMaxResult_18_#5:EAudioQATestResult Enum=Select(
     Option 0=NewEnumerator1
     Option 1=NewEnumerator0
     Index="In Range (Float)"(
      Value=@n18
      Min=@n17.MinShortTermLoudness_13_#20
      Max=@n17.MaxShortTermLoudness_14_#21
      InclusiveMin=true
      InclusiveMax=true
     )
    )
   ) -> (then, StructOut)
   n19 = Sequence(execute=@n16.then_1)
   n20 = Break Soundwave Validation Settings!(SoundwaveValidationSettings=Get Validation Settings()) -> (
    ValidateIntegratedLoudness_34_#35:bool=~
    MinIntegratedLoudness_37_#22:double=~
    MaxIntegratedLoudness_39_#23:double=~
   )
   > Branch(Condition=@n20)
   n21 = Break Sound Wave Validation Data!(SoundWaveValidationData=Get ValidationData()) -> (IntegratedLoudness_18_#13:double=~)
   > Set members in Sound Wave Validation Results!(
    execute
    StructRef=Get ValidationResults()
    IntegratedResult_19_#6:EAudioQATestResult Enum=Select(
     Option 0=NewEnumerator1
     Option 1=NewEnumerator0
     Index="In Range (Float)"(
      Value=@n21
      Min=@n20.MinIntegratedLoudness_37_#22
      Max=@n20.MaxIntegratedLoudness_39_#23
      InclusiveMin=true
      InclusiveMax=true
     )
    )
   ) -> (then, StructOut)
   n22 = Sequence(execute=@n19.then_1)
   > DEBUG_PrintTestResults!(
    execute
    self
    SoundWaveValidationData:Sound Wave Validation Data Structure=Get ValidationData()
    SoundWaveValidationResults:Sound Wave Validation Results Structure=Get ValidationResults()
   ) -> (then) {is=title K2Node_CallFunction}
   Return Node!(
    execute=@n22.then_1
    SoundWaveValidationData:Sound Wave Validation Data Structure=Get ValidationData()
    SoundWaveValidationResults:Sound Wave Validation Results Structure=Get ValidationResults()
   )
   Reroute Node(InputPin:"Map of Float (double-precision)s to Float (double-precision)s"=@n2.FramesWithFailingLoudnessValues) -> (OutputPin:"Map of Float (double-precision)s to Float (double-precision)s")
  ]}
 SetLoudnessAnalyzerSound {nodes [
   n1 = SetLoudnessAnalyzerSound!() -> (then, Object:Object Reference) {is=title K2Node_FunctionEntry}
   > n2 = Cast To SoundWave!(execute, Object=@n1) -> (then, CastFailed, AsSound Wave:Sound Wave Object Reference) {is=title K2Node_DynamicCast}
   > Is Valid(InputObject=@n2)
   > SetSound!(execute, self:Audio Analyzer NRT Object Reference=Get LoudnessNRT(), InSound:Sound Wave Object Reference=@n2) -> (then)
   > AnalyzeAudio(self:Audio Analyzer NRT Object Reference=Get LoudnessNRT())
  ]}
 ValidateNRTLoudnessRange {nodes [
   n1 = ValidateNRTLoudnessRange!() -> (
    then
    LoudnessNRT:LKFSNRT Object Reference
    LoudnessMin:double=-40
    LoudnessMax:double=-10
    MinTimeToValidate:double=~
    MaxTimeToValidate:double=~
    NumAllowedConsecutiveFailures:int=2
   ) {is=title K2Node_FunctionEntry}
   > n2 = Sequence()
   > Branch(Condition="float < float"(A=@n1.MaxTimeToValidate))
   > Return Node!(
    execute
    Result:EAudioQATestResult Enum=NewEnumerator2
    FramesWithFailingLoudnessValues:"Map of Float (double-precision)s to Float (double-precision)s"
   )
   Sequence(execute=@n2.then_1)
   > n3 = For Each Loop(Array=GetLoudnessData(self=@n1))
   > Branch(Condition="In Range (Float)"(
     Value=Break LKFSNRTResults!(LKFSNRTResults=@n3) -> (Timestamp:float=~)
     Min=@n1.MinTimeToValidate
     Max=@n1.MaxTimeToValidate
     InclusiveMin=true
     InclusiveMax=true
    ))
   n4 = Break LKFSNRTResults!(LKFSNRTResults=@n3) -> (Channel:int, Timestamp=~, Energy:float=~, Loudness:float=~, ShortTermLoudness:float=~)
   > n5 = Branch(Condition="In Range (Float)"(Value=@n4.ShortTermLoudness, Min=0, Max=1, InclusiveMin=true, InclusiveMax=true))
   > Set ConsecutiveFailures(ConsecutiveFailures:int=0) -> (Output_Get:int=0)
   Add(
    execute=@n5.else
    TargetMap:"Map of Float (double-precision)s to Float (double-precision)s"=Get TimeToFailingLoudnessValues() -> (TimeToFailingLoudnessValues:"Map of Float (double-precision)s to Float (double-precision)s")
    Key:"Float (double-precision) (by ref)"=@n4.Timestamp
    Value:"Float (double-precision) (by ref)"=@n4.ShortTermLoudness
   ) {is=title K2Node_CallFunction}
   > n6 = Increment Int(Value=Get ConsecutiveFailures() -> (ConsecutiveFailures:int=0))
   > Branch(Condition="integer > integer"(A=@n6, B=Reroute Node(InputPin:int) -> (OutputPin:int)))
   > Return Node(Result=NewEnumerator1, FramesWithFailingLoudnessValues=Get TimeToFailingLoudnessValues())
   Return Node(execute=@n3.Completed, Result=NewEnumerator0, FramesWithFailingLoudnessValues=Get TimeToFailingLoudnessValues())
  ]}
 ValidateSelectedSoundWaves {nodes [
   n1 = ValidateSelectedSoundWaves!() -> (then, ValidationSettings:BP Soundwave Validation Settings Object Reference) {is=title K2Node_FunctionEntry}
   > ValidateSelectedSoundWaves_Event!(execute, self, ValidationSettings:BP Soundwave Validation Settings Object Reference=@n1) -> (then) {is=title K2Node_CallFunction}
  ]}
 WriteTestResultsToCSV {nodes [
   n1 = WriteTestResultsToCSV!() -> (then, SoundWaveDataToResults:Map of Sound Wave Validation Data Structures to Sound Wave Validation Results Structures) {is=title K2Node_FunctionEntry}
   > n2 = "For Each Loop (Map)"(MapPin=@n1)
   n3 = Break Sound Wave Validation Data!(SoundWaveValidationData=@n2) -> (
    SoundWaveName_21_#7:String
    NumChannels_13_#8:int
    SampleRate_14_#9:int
    Duration_15_#10:double=~
    MomentaryLoudnessMax_16_#11:double=~
    ShortTermLoudnessMax_17_#12:double=~
    IntegratedLoudness_18_#13=~
   )
   > Add(TargetArray:Array of Strings=Get SoundWaveNames() -> (SoundWaveNames:Array of Strings), NewItem:"String (by ref)"=@n3) {is=title K2Node_CallArrayFunction}
   > Add(
    TargetArray:Array of Integers=Get NumChannelsValues() -> (NumChannelsValues:Array of Integers)
    NewItem:"Integer (by ref)"=@n3.NumChannels_13_#8
   )
   > Add(TargetArray=Get SampleRateValues() -> (SampleRateValues:Array of Integers), NewItem=@n3.SampleRate_14_#9)
   > Add(
    TargetArray:"Array of Float (double-precision)s"=Get DurationValues() -> (DurationValues:"Array of Float (double-precision)s")
    NewItem:"Float (double-precision) (by ref)"=@n3.Duration_15_#10
   )
   > Add(
    TargetArray=Get MomentaryMaxValues() -> (MomentaryMaxValues:"Array of Float (double-precision)s")
    NewItem=@n3.MomentaryLoudnessMax_16_#11
   )
   > Add(
    TargetArray=Get ShortTermMaxValues() -> (ShortTermMaxValues:"Array of Float (double-precision)s")
    NewItem=@n3.ShortTermLoudnessMax_17_#12
   )
   > Add(
    TargetArray=Get IntegratedValues() -> (IntegratedValues:"Array of Float (double-precision)s")
    NewItem=@n3.IntegratedLoudness_18_#13
   )
   n4 = Break Sound Wave Validation Results(SoundWaveValidationResults=@n2.ValuePin)
   > Add(
    TargetArray:Array of Strings=Get NumChannelsResults() -> (NumChannelsResults:Array of Strings)
    NewItem:"String (by ref)"=Enum to String(Enumerator=@n4)
   )
   > Add(
    TargetArray=Get SampleRateResults() -> (SampleRateResults:Array of Strings)
    NewItem=Enum to String(Enumerator=@n4.SampleRateResult_15_#2)
   )
   > Add(
    TargetArray=Get DurationResults() -> (DurationResults:Array of Strings)
    NewItem=Enum to String(Enumerator=@n4.DurationResult_16_#3)
   )
   > Add(
    TargetArray=Get MomentaryMaxResults() -> (MomentaryMaxResults:Array of Strings)
    NewItem=Enum to String(Enumerator=@n4.MomentaryLoudnessMaxResult_17_#4)
   )
   > Add(
    TargetArray=Get ShortTermMaxResults() -> (ShortTermMaxResults:Array of Strings)
    NewItem=Enum to String(Enumerator=@n4.ShortTermLoudnessMaxResult_18_#5)
   )
   > Add(
    TargetArray=Get IntegratedResults() -> (IntegratedResults:Array of Strings)
    NewItem=Enum to String(Enumerator=@n4.IntegratedResult_19_#6)
   )
   n5 = MakeCsvFilePath!(execute=@n2.CompletedPin, self) -> (then, Path:String) {is=title K2Node_CallFunction}
   > Execute Python Script!(
    execute
    PythonScript="import csv\r\nimport os\r\nimport sys\r\n\r\nprint(f\"Creating CSV file: '{csv_file_path}'\") ##type:ignore\r\n\r\nwith open(csv_file_path, \"w\") as CsvFile:\r\n    Writer = csv.writer(CsvFile)\r\n    Writer.writerow([\"SoundWaveName\", \"NumChannels\", \"SampleRate\", \"Duration\", \"LKFS-Momentary(Max)\", \"LKFS-ShortTerm(Max)\", \"LKFS-Integrated\"])\r\n    for i, soundwave in enumerate(soundwave_names): ##type: ignore\r\n        try:\r\n            Writer.writerow([f\"{soundwave}\", f\"{numchannels_values[i]} ({numchannels_results[i]})\", f\"{samplerate_values[i]} ({samplerate_results[i]})\", f\"{duration_values[i]:.2f} ({duration_results[i]})\", f\"{momentarymax_values[i]:.2f} ({momentarymax_results[i]})\", f\"{shorttermmax_values[i]:.2f} ({shorttermmax_results[i]})\", f\"{integrated_values[i]:.2f} ({integrated_results[i]})\"]) ##type: ignore\r\n        except IndexError:\r\n            print(\"Attempted to read loudness data, but index is out of range.\")\r\n            Writer.writerow([f\"{soundwave}\"]) ##type: ignore\r\n"
    csv_file_path:String=@n5
    soundwave_names:Array of Strings=Get SoundWaveNames()
    numchannels_values:Array of Integers=Get NumChannelsValues()
    samplerate_values:Array of Integers=Get SampleRateValues()
    duration_values:"Array of Float (double-precision)s"=Get DurationValues()
    momentarymax_values:"Array of Float (double-precision)s"=Get MomentaryMaxValues()
    shorttermmax_values:"Array of Float (double-precision)s"=Get ShortTermMaxValues()
    integrated_values:"Array of Float (double-precision)s"=Get IntegratedValues()
    numchannels_results:Array of Strings=Get NumChannelsResults()
    samplerate_results:Array of Strings=Get SampleRateResults()
    duration_results:Array of Strings=Get DurationResults()
    momentarymax_results:Array of Strings=Get MomentaryMaxResults()
    shorttermmax_results:Array of Strings=Get ShortTermMaxResults()
    integrated_results:Array of Strings=Get IntegratedResults()
   ) -> (then, ReturnValue) {inputs [
     csv_file_path
     soundwave_names
     numchannels_values
     samplerate_values
     duration_values
     momentarymax_values
     shorttermmax_values
     integrated_values
     numchannels_results
     samplerate_results
     duration_results
     momentarymax_results
     shorttermmax_results
     integrated_results
    ]}
  ]}
]
variables [
 CSVFile Name {type=string, default=SoundWaveValidation}
 CSVFileOutputPath {type=string, default=""}
 CurrentSoundwaveIndex {type=integer}
 LoudnessNRT {type=LKFSNRT, default=null}
 SelectedAssets {type="array<Object>", default []}
 Sound Wave Validation Results {}
 SoundWaveSummary {type=string, default=""}
 Validation Settings {
  type=SoundwaveValidationSettings
  default {
   csvFilenameBase=SoundWaveValidation
   csvFolder="D;\\Dev\\Python\\"
   makeCsvFile=true
   maxAnalysisTime=0
   maxDuration=1.5
   maxIntegratedLoudness=0
   maxMomentaryLoudness=-10
   maxShortTermLoudness=-8
   minAnalysisTime=0
   minDuration=0
   minIntegratedLoudness=-96
   minMomentaryLoudness=-96
   minShortTermLoudness=-50
   numAllowedConsecutiveShortTermLoudnessFailures=1
   requiredNumChannels=1
   requiredSampleRate=48000
   validateDuration=true
   validateIntegratedLoudness=true
   validateMomentaryLoudness=true
   validateNumChannels=true
   validateSampleRate=true
   validateShortTermLoudness=true
  }
 }
]
